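# One-shot stereo MetaSound for the GrindRail coaster tunnel whoosh, built from jump whoosh samples.
# After a short random delay on play, it mixes a random DeadAir wave at lower gain with a random Whoosh wave, finishing when both end.
# WhooshVol drives both output gain and pitch, with the Whoosh layer pitched down on the Default page.
# The Mobile page uses fewer waves, a different pitch range and no layer offset.
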MetaSoundSource /GrindRail/Audio/Coaster/MSS_GrindRail_Coaster_TunnelWhoosh
inputs [
 DeadAir {type=WaveAsset:Array}
 UE.Source.OnPlay {type=Trigger}
 Whoosh {type=WaveAsset:Array}
 WhooshVol {type=Float}
]
interfaces [UE.OutputFormat.Stereo:1.0, UE.Source.OneShot:1.0, UE.Source:1.1]
outputs [
 "UE.OutputFormat.Stereo.Audio:0" {type=Audio}
 "UE.OutputFormat.Stereo.Audio:1" {type=Audio}
 UE.Source.OneShot.OnFinished {type=Trigger}
]
pages [
 Default {
  is_build_graph=true
  input_defaults {
   UE.Source.OnPlay=null
   WhooshVol=1
   DeadAir [
    &/GrindRail_Jump_Whoosh_Dead_Air_01
    &/GrindRail_Jump_Whoosh_Dead_Air_02
    &/GrindRail_Jump_Whoosh_Dead_Air_03
   ]
   Whoosh [
    &/GrindRail_Jump_Whoosh_Air_01
    &/GrindRail_Jump_Whoosh_Air_02
    &/GrindRail_Jump_Whoosh_Air_03
    &/GrindRail_Jump_Whoosh_Air_04
   ]
  }
  nodes [
   n1 = UE.RandomTime(Next=^UE.Source.OnPlay) {also {defaults {Max=0.1}}}
   TriggerDelay_0 = UE.Trigger Delay(In=@n1.On Next, Delay Time=@n1.Value)
   ArrayRandomGet_0 = Array.Random Get:WaveAsset:Array(In Array=^DeadAir, Next=@TriggerDelay_0.Out) {also {defaults {Weights []}}}
   n2 = MapRange.MapRange:Float(In=^WhooshVol) {also {defaults {Out Range A=-2, Out Range B=2}}}
   WavePlayer_0 = UE.Wave Player:Stereo(Play=@ArrayRandomGet_0.On Next, Wave Asset=@ArrayRandomGet_0.Value, Pitch Shift=@n2.Out Value) {also {version [1,1]}}
   ArrayRandomGet_1 = Array.Random Get:WaveAsset:Array(In Array=^Whoosh, Next=@TriggerDelay_0.Out)
   WavePlayer_1 = UE.Wave Player:Stereo(
    Pitch Shift=UE.Add:Float(AdditionalOperands=@n2.Out Value) {also {defaults {PrimaryOperand=-6}}}.Out
    Play=@ArrayRandomGet_1.On Next
    Wave Asset=@ArrayRandomGet_1.Value
   ) {also {version [1,1]}}
   AudioMixerAudioMixerStereo2_0 = "AudioMixer.Audio Mixer (Stereo, 2)"(In 1 L=@WavePlayer_0.Out Left, In 1 R=@WavePlayer_0.Out Right, In 0 L=@WavePlayer_1.Out Left, In 0 R=@WavePlayer_1.Out Right) {also {defaults {Gain 1=0.5}}}
   n3 = UE.Reroute(Value=^WhooshVol)
   Multiply_0 = UE.Multiply:Audio by Float(PrimaryOperand=@AudioMixerAudioMixerStereo2_0.Out R, AdditionalOperands=@n3.Value)
   Multiply_1 = UE.Multiply:Audio by Float(PrimaryOperand=@AudioMixerAudioMixerStereo2_0.Out L, AdditionalOperands=@n3.Value)
   TriggerAccumulatorTriggerAccumulate2_0 = "TriggerAccumulator.Trigger Accumulate (2)"(In 1=@WavePlayer_0.On Finished, In 0=@WavePlayer_1.On Finished)
  ]
  links [
   {dst="@UE.OutputFormat.Stereo.Audio:1", src=Multiply_0.Out}
   {dst="@UE.OutputFormat.Stereo.Audio:0", src=Multiply_1.Out}
   {dst="@UE.Source.OneShot.OnFinished", src=TriggerAccumulatorTriggerAccumulate2_0.Out}
  ]
 }
 Mobile {
  input_defaults {
   DeadAir [&GrindRail_Jump_Whoosh_Dead_Air_01]
   Whoosh [&GrindRail_Jump_Whoosh_Air_01, &GrindRail_Jump_Whoosh_Air_02]
  }
  nodes [
   n1 = UE.RandomTime(Next=^UE.Source.OnPlay) {also {defaults {Max=0.1}}}
   TriggerDelay_0 = UE.Trigger Delay(In=@n1.On Next, Delay Time=@n1.Value)
   ArrayRandomGet_0 = Array.Random Get:WaveAsset:Array(In Array=^DeadAir, Next=@TriggerDelay_0.Out) {also {defaults {Weights []}}}
   n2 = MapRange.MapRange:Float(In=^WhooshVol) {also {defaults {Out Range B=4}}}
   WavePlayer_0 = UE.Wave Player:Stereo(Play=@ArrayRandomGet_0.On Next, Wave Asset=@ArrayRandomGet_0.Value, Pitch Shift=@n2.Out Value) {also {version [1,1]}}
   ArrayRandomGet_1 = Array.Random Get:WaveAsset:Array(In Array=^Whoosh, Next=@TriggerDelay_0.Out)
   WavePlayer_1 = UE.Wave Player:Stereo(Play=@ArrayRandomGet_1.On Next, Wave Asset=@ArrayRandomGet_1.Value, Pitch Shift=@n2.Out Value) {also {version [1,1]}}
   AudioMixerAudioMixerStereo2_0 = "AudioMixer.Audio Mixer (Stereo, 2)"(In 1 L=@WavePlayer_0.Out Left, In 1 R=@WavePlayer_0.Out Right, In 0 L=@WavePlayer_1.Out Left, In 0 R=@WavePlayer_1.Out Right) {also {defaults {Gain 1=0.5}}}
   n3 = UE.Reroute(Value=^WhooshVol)
   Multiply_0 = UE.Multiply:Audio by Float(PrimaryOperand=@AudioMixerAudioMixerStereo2_0.Out R, AdditionalOperands=@n3.Value)
   Multiply_1 = UE.Multiply:Audio by Float(PrimaryOperand=@AudioMixerAudioMixerStereo2_0.Out L, AdditionalOperands=@n3.Value)
   TriggerAccumulatorTriggerAccumulate2_0 = "TriggerAccumulator.Trigger Accumulate (2)"(In 1=@WavePlayer_0.On Finished, In 0=@WavePlayer_1.On Finished)
  ]
  links [
   {dst="@UE.OutputFormat.Stereo.Audio:1", src=Multiply_0.Out}
   {dst="@UE.OutputFormat.Stereo.Audio:0", src=Multiply_1.Out}
   {dst="@UE.Source.OneShot.OnFinished", src=TriggerAccumulatorTriggerAccumulate2_0.Out}
  ]
 }
]
properties {
 assetImportData=&self:AssetImportData
 attenuationSettings=/GrindRail/Audio/Grinding_1P_Att
 maxDistance=3000
 numChannels=2
 outputFormat=Stereo
 sampleRate=48000
 soundClassObject=&Fort_Player_Foley_Athena
 volume=0.5
}
