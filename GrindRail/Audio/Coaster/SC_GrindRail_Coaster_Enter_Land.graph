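# Sound cue for landing on a coaster grind rail, playing a metal impact on the player foley sound class.
# An IsLocal branch selects first-person attenuation for the local player and third-person attenuation when false or unset.
# Each path randomly picks one of two equally weighted landing waves and applies modulation with slight pitch variation.
# Both random nodes are excluded from branch culling.
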SoundCue /GrindRail/Audio/Coaster/SC_GrindRail_Coaster_Enter_Land
root=Branch_3
nodes [
 {id=Attenuation_6, children [Modulator_0], properties {attenuationSettings=/GrindRail/Audio/Grinding_1P_Att}}
 {id=Attenuation_7, children [Modulator_1], properties {attenuationSettings=/GrindRail/Audio/Grinding_3P_Att}}
 {id=Branch_3, children [Attenuation_6, Attenuation_7, Attenuation_7], properties {boolParameterName=IsLocal}}
 {id=Modulator_0, children [Random_0], properties {pitchMin=0.98}}
 {id=Modulator_1, children [Random_2], properties {pitchMin=0.98}}
 {id=Random_0, children [WavePlayer_2,WavePlayer_3], properties {bShouldExcludeFromBranchCulling=true, weights [1,1]}}
 {id=Random_2, children [WavePlayer_6,WavePlayer_7], properties {bShouldExcludeFromBranchCulling=true, weights [1,1]}}
 {id=WavePlayer_2, properties {soundWaveAssetPtr=&/Grind_Rail_Coaster_LandOnRailMetal_02}}
 {id=WavePlayer_3, properties {soundWaveAssetPtr=&/Grind_Rail_Coaster_LandOnRailMetal_01}}
 {id=WavePlayer_6, properties {soundWaveAssetPtr=&Grind_Rail_Coaster_LandOnRailMetal_02}}
 {id=WavePlayer_7, properties {soundWaveAssetPtr=&Grind_Rail_Coaster_LandOnRailMetal_01}}
]
properties {maxDistance=6500, soundClassObject=&Fort_Player_Foley_Athena, volumeMultiplier=0.64}
