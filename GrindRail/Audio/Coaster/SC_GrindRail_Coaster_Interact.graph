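# Sound cue for coaster grind rail interaction, layering rail jump-off and landing thud sounds.
# A Mixer combines a random jump-off wave with a random coaster thud, whose pitch an Enveloper sweeps upward.
# A Modulator adds slight pitch variation to the mix, and the cue's pitch multiplier doubles it overall.
# It uses item-interact attenuation and a concurrency override allowing one instance per owner, blocking new plays.
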SoundCue /GrindRail/Audio/Coaster/SC_GrindRail_Coaster_Interact
root=Modulator_4
nodes [
 {
  id=Enveloper_0
  children [Random_1]
  properties {pitchCurve {editorCurveData {keys [<at 0 0.5>,<at 1 2.5>]}}, volumeCurve {editorCurveData {keys [<at 0 1>]}}}
 }
 {id=Mixer_0, children [Random_6,Enveloper_0], properties {inputVolume [1,1]}}
 {id=Modulator_4, children [Mixer_0], properties {pitchMax=1.1, pitchMin=0.9}}
 {id=Random_1, children [WavePlayer_10, WavePlayer_11, WavePlayer_12], properties {weights [1,1,1]}}
 {id=Random_6, children [WavePlayer_0,WavePlayer_1,WavePlayer_2], properties {bShouldExcludeFromBranchCulling=true, weights [1,1,1]}}
 {id=WavePlayer_0, properties {soundWaveAssetPtr=/GrindRail/Audio/Grind_Rail_Jump_Off_Rail_01}}
 {id=WavePlayer_1, properties {soundWaveAssetPtr=/GrindRail/Audio/Grind_Rail_Jump_Off_Rail_02}}
 {id=WavePlayer_10, properties {soundWaveAssetPtr=&/Grind_Rail_Coaster_JumpOnLandThud_01}}
 {id=WavePlayer_11, properties {soundWaveAssetPtr=&/Grind_Rail_Coaster_JumpOnLandThud_02}}
 {id=WavePlayer_12, properties {soundWaveAssetPtr=&/Grind_Rail_Coaster_JumpOnLandThud_03}}
 {id=WavePlayer_2, properties {soundWaveAssetPtr=/GrindRail/Audio/Grind_Rail_Jump_Off_Rail_03}}
]
properties {
 attenuationSettings=/Game/Sounds/Attenuation/Fort_Item_Interact_Att
 bOverrideConcurrency=true
 maxDistance=1850
 pitchMultiplier=2
 soundClassObject=&Fort_UI_InGame_NonCritical
 volumeMultiplier=0.7
 concurrencyOverrides {bLimitToOwner=true, maxCount=1, resolutionRule=PreventNew}
 concurrencySet [/GrindRail/Audio/SC_GrindRail_Jump_SC]
}
