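# Sound cue for landing on a coaster grind rail after a jump.
# An IsLocal branch picks first-person attenuation when true and third-person attenuation when false or unset.
# Both paths share the same waves, mixing a random energy-connect sound with a louder-weighted random landing thud, each layer passing through a modulator for variation.
# The cue uses player foley sound class and a grind rail jump concurrency set.
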SoundCue /GrindRail/Audio/Coaster/SC_GrindRail_Coaster_Jump_Land
root=Branch_0
nodes [
 {id=Attenuation_0, children [Mixer_1], properties {attenuationSettings=/GrindRail/Audio/Grinding_1P_Att}}
 {id=Attenuation_1, children [Mixer_2], properties {attenuationSettings=/GrindRail/Audio/Grinding_3P_Att}}
 {id=Branch_0, children [Attenuation_0, Attenuation_1, Attenuation_1], properties {boolParameterName=IsLocal}}
 {id=Mixer_1, children [Modulator_1,Modulator_3], properties {inputVolume [1,1.5]}}
 {id=Mixer_2, children [Modulator_4,Modulator_5], properties {inputVolume [1,1.5]}}
 {id=Modulator_1, children [Random_1], properties {volumeMax=1.2}}
 {id=Modulator_3, children [Random_0]}
 {id=Modulator_4, children [Random_5], properties {volumeMax=1.2}}
 {id=Modulator_5, children [Random_2]}
 {id=Random_0, children [WavePlayer_0,WavePlayer_1,WavePlayer_2], properties {weights [1,1,1]}}
 {id=Random_1, children [WavePlayer_4,WavePlayer_5], properties {weights [1,1]}}
 {id=Random_2, children [WavePlayer_3, WavePlayer_14, WavePlayer_15], properties {weights [1,1,1]}}
 {id=Random_5, children [WavePlayer_9, WavePlayer_10], properties {weights [1,1]}}
 {id=WavePlayer_0, properties {soundWaveAssetPtr=&/Grind_Rail_Coaster_JumpOnLandThud_01}}
 {id=WavePlayer_1, properties {soundWaveAssetPtr=&/Grind_Rail_Coaster_JumpOnLandThud_02}}
 {id=WavePlayer_10, properties {soundWaveAssetPtr=/GrindRail/Audio/Grind_Rail_Jump_Land_Energy_Connect_02}}
 {id=WavePlayer_14, properties {soundWaveAssetPtr=&Grind_Rail_Coaster_JumpOnLandThud_02}}
 {id=WavePlayer_15, properties {soundWaveAssetPtr=&/Grind_Rail_Coaster_JumpOnLandThud_03}}
 {id=WavePlayer_2, properties {soundWaveAssetPtr=&Grind_Rail_Coaster_JumpOnLandThud_03}}
 {id=WavePlayer_3, properties {soundWaveAssetPtr=&Grind_Rail_Coaster_JumpOnLandThud_01}}
 {id=WavePlayer_4, properties {soundWaveAssetPtr=/GrindRail/Audio/Grind_Rail_Jump_Land_Energy_Connect_01}}
 {id=WavePlayer_5, properties {soundWaveAssetPtr=&Grind_Rail_Jump_Land_Energy_Connect_02}}
 {id=WavePlayer_9, properties {soundWaveAssetPtr=&Grind_Rail_Jump_Land_Energy_Connect_01}}
]
properties {
 maxDistance=6500
 pitchMultiplier=1.3
 soundClassObject=&Fort_Player_Foley_Athena
 volumeMultiplier=0.4
 concurrencySet [/GrindRail/Audio/SC_GrindRail_Jump_SC]
}
