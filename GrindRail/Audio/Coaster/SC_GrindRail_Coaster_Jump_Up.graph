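# Sound cue for jumping up off a coaster grind rail, classed as player foley.
# An IsLocal branch selects first-person attenuation for the local player and third-person attenuation when false or unset.
# Both paths use the same waves, mixing a random generic rail jump-off with a louder random coaster jump-off, then randomly raising the pitch slightly.
# A jump concurrency asset governs overlapping plays.
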SoundCue /GrindRail/Audio/Coaster/SC_GrindRail_Coaster_Jump_Up
root=Branch_0
nodes [
 {id=Attenuation_0, children [Modulator_4], properties {attenuationSettings=/GrindRail/Audio/Grinding_1P_Att}}
 {id=Attenuation_1, children [Modulator_3], properties {attenuationSettings=/GrindRail/Audio/Grinding_3P_Att}}
 {id=Branch_0, children [Attenuation_0, Attenuation_1, Attenuation_1], properties {boolParameterName=IsLocal}}
 {id=Mixer_2, children [Random_6,Random_8], properties {inputVolume [1,1.5]}}
 {id=Mixer_3, children [Random_7,Random_0], properties {inputVolume [1,1.5]}}
 {id=Modulator_3, children [Mixer_3], properties {pitchMax=1.2, pitchMin=1.1}}
 {id=Modulator_4, children [Mixer_2], properties {pitchMax=1.2, pitchMin=1.1}}
 {id=Random_0, children [WavePlayer_7,WavePlayer_8,WavePlayer_9], properties {bShouldExcludeFromBranchCulling=true, weights [1,1,1]}}
 {
  id=Random_6
  children [WavePlayer_0,WavePlayer_1,WavePlayer_2,WavePlayer_4]
  properties {bShouldExcludeFromBranchCulling=true, weights [1,1,1,1]}
 }
 {
  id=Random_7
  children [WavePlayer_36, WavePlayer_37, WavePlayer_38, WavePlayer_39]
  properties {bShouldExcludeFromBranchCulling=true, weights [1,1,1,1]}
 }
 {id=Random_8, children [WavePlayer_3,WavePlayer_5,WavePlayer_6], properties {bShouldExcludeFromBranchCulling=true, weights [1,1,1]}}
 {id=WavePlayer_0, properties {soundWaveAssetPtr=/GrindRail/Audio/Grind_Rail_Jump_Off_Rail_01}}
 {id=WavePlayer_1, properties {soundWaveAssetPtr=/GrindRail/Audio/Grind_Rail_Jump_Off_Rail_02}}
 {id=WavePlayer_2, properties {soundWaveAssetPtr=/GrindRail/Audio/Grind_Rail_Jump_Off_Rail_03}}
 {id=WavePlayer_3, properties {soundWaveAssetPtr=&/Grind_Rail_Coaster_JumpOffRail_02}}
 {id=WavePlayer_36, properties {soundWaveAssetPtr=&Grind_Rail_Jump_Off_Rail_01}}
 {id=WavePlayer_37, properties {soundWaveAssetPtr=&Grind_Rail_Jump_Off_Rail_02}}
 {id=WavePlayer_38, properties {soundWaveAssetPtr=&Grind_Rail_Jump_Off_Rail_03}}
 {id=WavePlayer_39, properties {soundWaveAssetPtr=/GrindRail/Audio/Grind_Rail_Jump_Off_Rail_04}}
 {id=WavePlayer_4, properties {soundWaveAssetPtr=&Grind_Rail_Jump_Off_Rail_04}}
 {id=WavePlayer_5, properties {soundWaveAssetPtr=&/Grind_Rail_Coaster_JumpOffRail_01}}
 {id=WavePlayer_6, properties {soundWaveAssetPtr=&/Grind_Rail_Coaster_JumpOffRail_03}}
 {id=WavePlayer_7, properties {soundWaveAssetPtr=&Grind_Rail_Coaster_JumpOffRail_02}}
 {id=WavePlayer_8, properties {soundWaveAssetPtr=&Grind_Rail_Coaster_JumpOffRail_01}}
 {id=WavePlayer_9, properties {soundWaveAssetPtr=&Grind_Rail_Coaster_JumpOffRail_03}}
]
properties {
 maxDistance=6500
 soundClassObject=&Fort_Player_Foley_Athena
 volumeMultiplier=0.49
 concurrencyOverrides {bLimitToOwner=true, maxCount=1}
 concurrencySet [/GrindRail/Audio/SC_GrindRail_Jump_SC]
}
