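# Sound cue for the coaster grind rail's StartLoop sound, routed to the gadgets sound class with playback concurrency limited per owner.
# A branch on the IsLocal parameter selects first-person attenuation when true and third-person attenuation when false or unset.
# Both paths apply the same short random delay and raised random pitch, then pick between two equally weighted StartLoop wave variants.
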SoundCue /GrindRail/Audio/Coaster/SC_GrindRail_Coaster_StartLoop
root=Branch_0
nodes [
 {id=Attenuation_0, children [Delay_0], properties {attenuationSettings=/GrindRail/Audio/Grinding_1P_Att}}
 {id=Attenuation_1, children [Delay_1], properties {attenuationSettings=/GrindRail/Audio/Grinding_3P_Att}}
 {id=Branch_0, children [Attenuation_0, Attenuation_1, Attenuation_1], properties {boolParameterName=IsLocal}}
 {id=Delay_0, children [Modulator_0], properties {delayMax=0.5, delayMin=0.4}}
 {id=Delay_1, children [Modulator_1], properties {delayMax=0.5, delayMin=0.4}}
 {id=Modulator_0, children [Random_0], properties {pitchMax=1.3, pitchMin=1.2}}
 {id=Modulator_1, children [Random_2], properties {pitchMax=1.3, pitchMin=1.2}}
 {id=Random_0, children [WavePlayer_4,WavePlayer_5], properties {bShouldExcludeFromBranchCulling=true, weights [1,1]}}
 {id=Random_2, children [WavePlayer_6,WavePlayer_7], properties {bShouldExcludeFromBranchCulling=true, weights [1,1]}}
 {id=WavePlayer_4, properties {soundWaveAssetPtr=&/Grind_Rail_Coaster_StartLoop_01}}
 {id=WavePlayer_5, properties {soundWaveAssetPtr=&/Grind_Rail_Coaster_StartLoop_02}}
 {id=WavePlayer_6, properties {soundWaveAssetPtr=&Grind_Rail_Coaster_StartLoop_01}}
 {id=WavePlayer_7, properties {soundWaveAssetPtr=&Grind_Rail_Coaster_StartLoop_02}}
]
properties {
 maxDistance=6500
 soundClassObject=&Fort_Gadgets
 volumeMultiplier=0.216
 attenuationOverrides {
  bEnableOcclusion=true
  lPFFrequencyAtMin=10000
  lPFRadiusMin=2000
  nonFocusVolumeAttenuation=0.7
  occlusionLowPassFilterFrequency=4000
  occlusionTraceChannel=ECC_GameTraceChannel11
  occlusionVolumeAttenuation=4
 }
 concurrencyOverrides {bLimitToOwner=true}
}
